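# Parent material for BadCat character face accessories on skeletal meshes, adding indexed eyes and a flipbook mouth.
# The shared 2023 PreFX function feeds the BadCat eye function, then the vertex-color flipbook mouth function with scaled and shifted UVs, then shared PostFX.
# A vertex-color test adds scaled base color to emissive wherever vertices fall below full white.
# A material-attribute blend applies a shading-model override after the mouth stage.
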Material /BRCosmetics/Characters/Player/Male/Large/Bodies/M_LRG_BadCat/Materials/M_BadCat_FaceAcc_Parent
graph {
 properties {
  bUseMaterialAttributes=true
  bUsedWithMeshDeformer=true
  bUsedWithMorphTargets=true
  bUsedWithSkeletalMesh=true
  pixelDepthOffsetMode=PDOM_Legacy
  shadingModel=MSM_FromMaterialExpression
  usedShadingModels=Default Lit | Subsurface | Clear Coat | Subsurface Profile | Cloth
 }
 nodes [
  Constant() {r=1}
  n1 = MaterialFunctionCall() {materialFunction=&MF_CharacterParent_2023_PreFX}
  n2 = GetMaterialAttributes(None=MaterialFunctionCall!(Material=@n1.Result) {materialFunction=&/MF_IndexEyes_BadCat}.Result) {attributeGetTypes [#1, #2]}
  n3 = Add(A=@n2.Emissive Color, B=Multiply(A=@n2.Base Color, B=ScalarParameter() {defaultValue=0.75, group=BadCat, parameterName=EyesEmissive_Amount}))
  n4 = Constant()
  n5 = If(A=DotProduct(A=VertexColor(), B=Constant3Vector() {constant=<rgba 1 1 1 0>}), "A > B"=@n4, "A == B"=@n4, "A < B"=Constant() {r=1}) {constB=3}
  n6 = Multiply(A=TextureCoordinate(), B=ScalarParameter() {defaultValue=2.5, group=09_Flipbook Mouth, parameterName=MouthScale})
  n7 = AppendVector(A=ScalarParameter() {defaultValue=1, group=09_Flipbook Mouth, parameterName=MouthScale_X}, B=Constant() {r=1})
  n8 = ScalarParameter() {defaultValue=-0.59, group=09_Flipbook Mouth, parameterName=MouthUVShiftX}
  n9 = ScalarParameter() {defaultValue=-1.5174, group=09_Flipbook Mouth, parameterName=MouthUVShiftY}
  n10 = MaterialFunctionCall!(
   Material=SetMaterialAttributes(MaterialAttributes=@n2.MaterialAttributes, Emissive Color=LinearInterpolate(A=@n2.Emissive Color, B=@n3, Alpha=@n5)) {attributeSetTypes [#1]}
   MouthUVs=Add(A=Multiply(A=@n6, B=@n7), B=AppendVector(A=@n8, B=@n9))
  ) {materialFunction=/Game/Characters/Player/Common/MaterialFunctions/VertexColorFaces/MF_FV2_FlipbookMouths}
  n11 = ShadingModel()
  MaterialFunctionCall(
   Material=BlendMaterialAttributes(
    A=@n10.Result
    B=SetMaterialAttributes!(MaterialAttributes=@n10.Result, Shading Model=@n11) {attributeSetTypes [#3]}
    Alpha=@n10.Result
   )
   M=@n1.M
   Pre Skinned Local Position=@n1.M
   Distance Blend=@n1.M
   Fresnel=@n1.M
  ) {materialFunction=&MF_CharacterParent_2023_PostFX}
 ]
}
