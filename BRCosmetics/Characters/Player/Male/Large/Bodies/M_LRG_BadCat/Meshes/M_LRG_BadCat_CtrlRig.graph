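# Control Rig for the large male BadCat character body that drives costume and mechanical bones from the animated skeleton.
# A shared DeformRig function runs first, then a sequence handles spine mechanism, toes, collars, bracelets, leg extras and armor.
# Spine mech uses isolated rotation and aim constraints, including mutually aimed joint pairs.
# Spherical pose readers, some interpolated, weight corrective offsets; elbow armor also reads animation attributes, and `skydiving` scales a right-collar correction.
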ControlRigBlueprint /BRCosmetics/Characters/Player/Male/Large/Bodies/M_LRG_BadCat/Meshes/M_LRG_BadCat_CtrlRig
graphs [
 RigVMFunctionLibrary {links [], nodes []}
 RigVMFunction_Sequence_1_ContainedGraph {nodes [
   n1 = RigVMFunctionEntryNode()
   RigVMFunction_Sequence = &RigVMFunction_Sequence(ExecuteContext=@n1.ExecuteContext)
   n2 = &RigVMFunction_Sequence(ExecuteContext=@RigVMFunction_Sequence.B)
   n3 = &RigVMFunction_Sequence(ExecuteContext=@n2.B)
   n4 = &RigVMFunction_Sequence(ExecuteContext=@n3.B)
   n5 = &RigVMFunction_Sequence(ExecuteContext=@n4.B)
   n6 = &RigVMFunction_Sequence(ExecuteContext=@n5.B)
   n7 = &RigVMFunction_Sequence(ExecuteContext=@n6.B)
   RigVMFunctionReturnNode(A=@RigVMFunction_Sequence.A, B=@n2.A, C=@n3.A, D=@n4.A, E=@n5.A, F=@n6.A, G=@n7.A, H=@n7.B)
  ]}
 RigVMModel {nodes [
   RigUnit_BeginExecution = &RigUnit_BeginExecution()
   FN_DeformRig_FNC = RigVMFunctionReferenceNode(
    DeformRig_ Curve="(deformRig_l_upperarm_up_alpha_72_#1=False,deformRig_r_upperarm_up_alpha_73_#2=False,deformRig_l_upperarm_fr_alpha_74_#3=False,deformRig_r_upperarm_fr_alpha_75_CB51B8 ...(truncated)"
    DeformRig_Bool="(upperarm_1_#4=True,elbow_3_#5=True,wrist_7_#6=True,groin_8_#7=True,glute_17_#8=True,knee_9_40C408324CA261 ...(truncated)"
    DeformRig_Debug=(upperarm_debug_18_#4=False,wrist_debug_20_#6=False,lat_debug_26_#9=False,groinGlute_debug_27_#7=False)
    DeformRig_Value="(upperarm_up_18_#10=(Rotation=(X=0,Y=0,Z=0,W=1),Translation=(X=0,Y=0,Z=8),Scale3D=(X=0.7,Y=1,Z=1)),upperarm_fwd_fr_19_#11=(Rotation=(X=0,Y=0,Z=0,W=1),Translation=(X=0,Y=10,Z=0),Scale3D=(X=0.5,Y=1,Z= ...(truncated)"
    ExecuteContext=@RigUnit_BeginExecution.ExecutePin
   )
   > RigVMFunction_Sequence = RigVMAggregateNode()
   RigVMFunction_MathQuaternionToEuler = &RigVMFunction_MathQuaternionToEuler(
    RotationOrder=XYZ
    Value=&RigUnit_GetTransform(Item=(Type=Bone,Name="spine_02"), Space=LocalSpace, bInitial=False).Transform.Rotation
   )
   RigVMFunction_MathQuaternionFromEuler = &RigVMFunction_MathQuaternionFromEuler(Euler=(X=0,Y=0,Z=0), RotationOrder=XYZ, Euler.X=@RigVMFunction_MathQuaternionToEuler.Result.X)
   &RigUnit_ModifyTransforms(
    ItemToModify=((Item=(Type=Bone,Name="pelvis_mech_turning"),Transform=(Rotation=(X=0,Y=0,Z=0,W=1),Translation=(X=0,Y=0,Z=0),Scale3D=(X=1,Y=1,Z=1))))
    Weight=1
    ItemToModify.0.Transform.Rotation=@RigVMFunction_MathQuaternionFromEuler.Result
    ExecutePin=@RigVMFunction_Sequence.A
   )
   > &RigUnit_AimConstraintLocalSpaceOffset(
    Child=(Type=Bone,Name="pelvis_mech_slant")
    Filter=(bX=false,bY=True,bZ=false)
    Parents=((Item=(Type=Bone,Name="spine_04"),Weight=1))
    UpAxis=(X=0,Y=1,Z=0)
    Weight=1
    WorldUp=(Target=(X=0,Y=0,Z=1),Kind=Direction,Space=(Type=None,Name="None"))
   )
   n1 = &RigUnit_GetTransform(Item=(Type=Bone,Name="spine_03"), Space=LocalSpace, bInitial=False)
   > &RigUnit_ModifyTransforms(
    ItemToModify=((Item=(Type=Bone,Name="wheel"),Transform=(Rotation=(X=0,Y=0,Z=0,W=1),Translation=(X=0,Y=0,Z=0),Scale3D=(X=1,Y=1,Z=1))))
    Weight=1
    ItemToModify.0.Transform.Rotation=&RigVMFunction_MathQuaternionFromEuler(Euler=(X=0,Y=0,Z=0), Euler.X=&RigVMFunction_MathQuaternionToEuler(RotationOrder=ZYX, Value=@n1.Transform.Rotation).Result.X).Result
   )
   > &RigUnit_AimConstraintLocalSpaceOffset(
    Child=(Type=Bone,Name="wheel")
    Filter=(bX=false,bY=false,bZ=True)
    Parents=((Item=(Type=Bone,Name="spine_05"),Weight=1))
    UpAxis=(X=0,Y=1,Z=0)
    Weight=1
    WorldUp=(Target=(X=0,Y=0,Z=1),Kind=Direction,Space=(Type=None,Name="None"))
   )
   n2 = &RigUnit_GetTransform(Item=(Type=Bone,Name="spine_04"), Space=LocalSpace, bInitial=False)
   > &RigUnit_ModifyTransforms(
    ItemToModify=((Item=(Type=Bone,Name="upper_body"),Transform=(Rotation=(X=0,Y=0,Z=0,W=1),Translation=(X=0,Y=0,Z=0),Scale3D=(X=1,Y=1,Z=1))))
    Weight=1
    ItemToModify.0.Transform.Rotation=&RigVMFunction_MathQuaternionFromEuler(Euler=(X=0,Y=0,Z=0), Euler.X=&RigVMFunction_MathQuaternionToEuler(RotationOrder=ZYX, Value=@n2.Transform.Rotation).Result.X).Result
   )
   > &RigUnit_AimConstraintLocalSpaceOffset(
    Child=(Type=Bone,Name="upper_body")
    Filter=(bX=false,bY=true,bZ=false)
    Parents=((Item=(Type=Bone,Name="spine_05"),Weight=1))
    UpAxis=(X=0,Y=1,Z=0)
    Weight=1
    WorldUp=(Target=(X=0,Y=0,Z=1),Kind=Direction,Space=(Type=None,Name="None"))
   )
   > &RigUnit_AimConstraintLocalSpaceOffset(
    AimAxis=(X=-1,Y=0,Z=0)
    Child=(Type=Bone,Name="joint8")
    Filter=(bX=false,bY=true,bZ=true)
    Parents=((Item=(Type=Bone,Name="joint7"),Weight=1))
    UpAxis=(X=0,Y=1,Z=0)
    Weight=1
    WorldUp=(Target=(X=0,Y=0,Z=1),Kind=Direction,Space=(Type=None,Name="None"))
   )
   > &RigUnit_AimConstraintLocalSpaceOffset(
    Child=(Type=Bone,Name="joint7")
    Filter=(bX=false,bY=true,bZ=true)
    Parents=((Item=(Type=Bone,Name="joint8"),Weight=1))
    UpAxis=(X=0,Y=1,Z=0)
    Weight=1
    WorldUp=(Target=(X=0,Y=0,Z=1),Kind=Direction,Space=(Type=None,Name="None"))
   )
   > &RigUnit_AimConstraintLocalSpaceOffset(
    Child=(Type=Bone,Name="joint10")
    Filter=(bX=false,bY=true,bZ=true)
    Parents=((Item=(Type=Bone,Name="joint9"),Weight=1))
    Weight=1
    WorldUp=(Target=(X=0,Y=1,Z=0),Kind=Direction,Space=(Type=None,Name="None"))
   )
   > &RigUnit_AimConstraintLocalSpaceOffset(
    AimAxis=(X=-1,Y=0,Z=0)
    Child=(Type=Bone,Name="joint9")
    Filter=(bX=false,bY=true,bZ=true)
    Parents=((Item=(Type=Bone,Name="joint10"),Weight=1))
    Weight=1
    WorldUp=(Target=(X=0,Y=1,Z=0),Kind=Direction,Space=(Type=None,Name="None"))
   )
   n3 = &RigUnit_SphericalPoseReader(
    ActiveRegionScaleFactors=(PositiveWidth=0.3,NegativeWidth=0.3,PositiveHeight=0.3,NegativeHeight=0.3)
    ActiveRegionSize=0.3
    Debug=(bDrawDebug=False,bDraw2D=false,bDrawLocalAxes=false,DebugScale=30,DebugSegments=30,DebugThickness=0.2)
    DriverAxis=(X=-1,Y=0,Z=0)
    DriverItem=(Type=Bone,Name="upperarm_r")
    FalloffRegionScaleFactors=(PositiveWidth=0.1,NegativeWidth=1,PositiveHeight=1,NegativeHeight=0.5)
    FalloffSize=0.5
    OptionalParentItem=(Type=Bone,Name="None")
    RotationOffset=(X=-40,Y=220,Z=0)
    ExecutePin=@RigVMFunction_Sequence.C
   )
   > &RigUnit_ModifyTransforms(
    ItemToModify=((Item=(Type=Bone,Name="deform_upperarm_fr_base_r"),Transform=(Rotation=(X=0,Y=0,Z=0,W=1),Translation=(X=30,Y=0,Z=0),Scale3D=(X=1,Y=1,Z=1))))
    Weight=@n3.OutputParam
   )
   > &RigUnit_ModifyTransforms(
    ItemToModify=((Item=(Type=Bone,Name="R_collar_01"),Transform=(Rotation=(X=0,Y=0,Z=0,W=1),Translation=(X=0,Y=0,Z=0),Scale3D=(X=0.5,Y=1,Z=1))))
    Weight=@n3.OutputParam
   )
   Remap = &RigVMFunction_MathDoubleRemap(
    SourceMaximum=36
    SourceMinimum=22
    TargetMaximum=1
    TargetMinimum=0
    Value=&RigVMFunction_MathQuaternionToEuler(
     RotationOrder=ZYX
     Value=&RigUnit_GetRelativeTransformForItem(Child=(Type=Bone,Name="spine_05"), Parent=(Type=Bone,Name="spine_01"), bChildInitial=False, bParentInitial=False).RelativeTransform.Rotation
    ).Result.X
   )
   > &RigUnit_ModifyTransforms(
    ItemToModify=((Item=(Type=Bone,Name="L_collar_01"),Transform=(Rotation=(X=0,Y=0,Z=0,W=1),Translation=(X=0,Y=-1,Z=0),Scale3D=(X=1,Y=1,Z=1))))
    Weight=&RigVMFunction_MathDoubleMul(B=@Remap.Result, A=@n3.OutputParam).Result
   )
   n4 = &RigUnit_SphericalPoseReader(
    ActiveRegionScaleFactors=(PositiveWidth=0.2,NegativeWidth=1,PositiveHeight=0.8,NegativeHeight=0.8)
    ActiveRegionSize=0.2
    Debug=(bDrawDebug=True,bDraw2D=false,bDrawLocalAxes=false,DebugScale=30,DebugSegments=30,DebugThickness=0.2)
    DriverAxis=(X=1,Y=0,Z=0)
    DriverItem=(Type=Bone,Name="clavicle_l")
    FalloffRegionScaleFactors=(PositiveWidth=0.2,NegativeWidth=1,PositiveHeight=0.5,NegativeHeight=0.5)
    FalloffSize=0.5
    OptionalParentItem=(Type=Bone,Name="None")
    RotationOffset=(X=0,Y=-20,Z=0)
    ExecutePin=@RigVMFunction_Sequence.H
   )
   > &RigUnit_ModifyTransforms(
    ItemToModify=((Item=(Type=Bone,Name="L_collar_03"),Transform=(Rotation=(X=0,Y=0,Z=0,W=1),Translation=(X=0,Y=-6,Z=0),Scale3D=(X=1,Y=1,Z=1))),(Item=(Type=Bone,Name="L_collar_02"),Transform=(Rotation=(X=0,Y=0,Z=0,W=1),Translation=(X=0,Y=-2,Z=0),Scale3D=(X=1,Y=1,Z=1))))
    Weight=@n4.OutputParam
   )
   > n5 = &RigUnit_SphericalPoseReader(
    ActiveRegionScaleFactors=(PositiveWidth=0.2,NegativeWidth=1,PositiveHeight=0.8,NegativeHeight=0.8)
    ActiveRegionSize=0.2
    Debug=(bDrawDebug=True,bDraw2D=false,bDrawLocalAxes=false,DebugScale=30,DebugSegments=30,DebugThickness=0.2)
    DriverAxis=(X=-1,Y=0,Z=0)
    DriverItem=(Type=Bone,Name="clavicle_r")
    FalloffRegionScaleFactors=(PositiveWidth=0.2,NegativeWidth=1,PositiveHeight=0.5,NegativeHeight=0.5)
    FalloffSize=0.5
    OptionalParentItem=(Type=Bone,Name="None")
    RotationOffset=(X=0,Y=160,Z=0)
   )
   > &RigUnit_ModifyTransforms(
    ItemToModify="((Item=(Type=Bone,Name=\"R_collar_03\"),Transform=(Rotation=(X=0,Y=0,Z=0,W=1),Translation=(X=0,Y=-4,Z=0),Scale3D=(X=1,Y=1,Z=1))),(Item=(Type=Bone,Name=\"R_collar_02\"),Transform=(Rotation=(X=0,Y=0,Z=0,W=1),Translation=(X=0,Y=-6,Z=0),Scale3D=(X=1,Y=1,Z=1))),(It ...(truncated)"
    Weight=@n5.OutputParam
   )
   &RigUnit_ParentConstraint(
    AdvancedSettings=(InterpolationType=Average,RotationOrderForFilter=XZY)
    Child=(Type=Bone,Name="L_bracelet")
    Filter=(TranslationFilter=(bX=true,bY=true,bZ=true),RotationFilter=(bX=true,bY=true,bZ=true),ScaleFilter=(bX=True,bY=True,bZ=True))
    Parents="((Item=(Type=Bone,Name=\"lowerarm_l\"),Weight=0.0228),(Item=(Type=Bone,Name=\"lowerarm_twist_01_l\"),Weight=0.6896),(Item=(Type=Bone,Name=\"hand_l\"),Weight=0.1719),(Item=(Type=Bone,Name=\"thumb_02_l\"),Weight=0.001),(Item=(Type=Bone,Name=\"thumb_01_l\"),Weight=0.04 ...(truncated)"
    Weight=1
    ExecutePin=@RigVMFunction_Sequence.D
   )
   > n6 = &RigUnit_SphericalPoseReader(
    ActiveRegionScaleFactors=(PositiveWidth=0.2,NegativeWidth=0.2,PositiveHeight=0.1,NegativeHeight=1)
    ActiveRegionSize=0.5
    Debug=(bDrawDebug=False,bDraw2D=false,bDrawLocalAxes=false,DebugScale=30,DebugSegments=30,DebugThickness=0.2)
    DriverAxis=(X=1,Y=0,Z=0)
    DriverItem=(Type=Bone,Name="hand_l")
    FalloffRegionScaleFactors=(PositiveWidth=0.5,NegativeWidth=0.5,PositiveHeight=1,NegativeHeight=1)
    FalloffSize=0.5
    OptionalParentItem=(Type=Bone,Name="None")
    RotationOffset=(X=0,Y=90,Z=-90)
   )
   > &RigUnit_ModifyTransforms(
    ItemToModify=((Item=(Type=Bone,Name="L_bracelet_upper"),Transform=(Rotation=(X=0,Y=0,Z=0,W=1),Translation=(X=-3,Y=0,Z=2),Scale3D=(X=1,Y=1,Z=1))))
    Weight=@n6.OutputParam
   )
   > n7 = &RigUnit_SphericalPoseReader(
    ActiveRegionScaleFactors=(PositiveWidth=0.2,NegativeWidth=0.2,PositiveHeight=0.1,NegativeHeight=1)
    ActiveRegionSize=0.5
    Debug=(bDrawDebug=False,bDraw2D=false,bDrawLocalAxes=false,DebugScale=30,DebugSegments=30,DebugThickness=0.2)
    DriverAxis=(X=1,Y=0,Z=0)
    DriverItem=(Type=Bone,Name="hand_l")
    FalloffRegionScaleFactors=(PositiveWidth=0.5,NegativeWidth=0.5,PositiveHeight=1,NegativeHeight=1)
    FalloffSize=0.5
    OptionalParentItem=(Type=Bone,Name="None")
    RotationOffset=(X=180,Y=0,Z=90)
   )
   > &RigUnit_ModifyTransforms(
    ItemToModify=((Item=(Type=Bone,Name="L_bracelet_r_side"),Transform=(Rotation=(X=0,Y=0,Z=0,W=1),Translation=(X=-3,Y=2,Z=0),Scale3D=(X=1,Y=1,Z=1))))
    Weight=@n7.OutputParam
   )
   > n8 = &RigUnit_SphericalPoseReader(
    ActiveRegionScaleFactors=(PositiveWidth=0.2,NegativeWidth=0.2,PositiveHeight=0.1,NegativeHeight=1)
    ActiveRegionSize=0.5
    Debug=(bDrawDebug=False,bDraw2D=false,bDrawLocalAxes=false,DebugScale=30,DebugSegments=30,DebugThickness=0.2)
    DriverAxis=(X=1,Y=0,Z=0)
    DriverItem=(Type=Bone,Name="hand_l")
    FalloffRegionScaleFactors=(PositiveWidth=0.5,NegativeWidth=0.5,PositiveHeight=1,NegativeHeight=1)
    FalloffSize=0.5
    OptionalParentItem=(Type=Bone,Name="None")
    RotationOffset=(X=180,Y=180,Z=90)
   )
   > &RigUnit_ModifyTransforms(
    ItemToModify=((Item=(Type=Bone,Name="L_bracelet_l_side"),Transform=(Rotation=(X=0,Y=0,Z=0,W=1),Translation=(X=-3,Y=-2,Z=0),Scale3D=(X=1,Y=1,Z=1))))
    Weight=@n8.OutputParam
   )
   > n9 = &RigUnit_SphericalPoseReader(
    ActiveRegionScaleFactors=(PositiveWidth=0.2,NegativeWidth=0.2,PositiveHeight=1,NegativeHeight=0.1)
    ActiveRegionSize=0.3
    Debug=(bDrawDebug=False,bDraw2D=false,bDrawLocalAxes=false,DebugScale=30,DebugSegments=30,DebugThickness=0.2)
    DriverAxis=(X=1,Y=0,Z=0)
    DriverItem=(Type=Bone,Name="hand_l")
    FalloffRegionScaleFactors=(PositiveWidth=0.5,NegativeWidth=0.5,PositiveHeight=1,NegativeHeight=1)
    FalloffSize=0.5
    OptionalParentItem=(Type=Bone,Name="None")
    RotationOffset=(X=180,Y=90,Z=-90)
   )
   > &RigUnit_ModifyTransforms(
    ItemToModify=((Item=(Type=Bone,Name="L_bracelet_lower"),Transform=(Rotation=(X=0,Y=0,Z=0,W=1),Translation=(X=-3,Y=0,Z=-2),Scale3D=(X=1,Y=1,Z=1))))
    Weight=@n9.OutputParam
   )
   > &RigUnit_ParentConstraint(
    AdvancedSettings=(InterpolationType=Average,RotationOrderForFilter=XZY)
    Child=(Type=Bone,Name="R_bracelet")
    Filter=(TranslationFilter=(bX=true,bY=true,bZ=true),RotationFilter=(bX=true,bY=true,bZ=true),ScaleFilter=(bX=True,bY=True,bZ=True))
    Parents="((Item=(Type=Bone,Name=\"lowerarm_r\"),Weight=0.0228),(Item=(Type=Bone,Name=\"lowerarm_twist_01_r\"),Weight=0.6896),(Item=(Type=Bone,Name=\"hand_r\"),Weight=0.1719),(Item=(Type=Bone,Name=\"thumb_02_r\"),Weight=0.001),(Item=(Type=Bone,Name=\"thumb_03_r\"),Weight=0.04 ...(truncated)"
    Weight=1
   )
   > n10 = &RigUnit_SphericalPoseReader(
    ActiveRegionScaleFactors=(PositiveWidth=0.2,NegativeWidth=0.2,PositiveHeight=0.1,NegativeHeight=1)
    ActiveRegionSize=0.5
    Debug=(bDrawDebug=False,bDraw2D=false,bDrawLocalAxes=false,DebugScale=30,DebugSegments=30,DebugThickness=0.2)
    DriverAxis=(X=-1,Y=0,Z=0)
    DriverItem=(Type=Bone,Name="hand_r")
    FalloffRegionScaleFactors=(PositiveWidth=0.5,NegativeWidth=0.5,PositiveHeight=1,NegativeHeight=1)
    FalloffSize=0.5
    OptionalParentItem=(Type=Bone,Name="None")
    RotationOffset=(X=0,Y=90,Z=90)
   )
   > &RigUnit_ModifyTransforms(
    ItemToModify=((Item=(Type=Bone,Name="L_bracelet_upper"),Transform=(Rotation=(X=0,Y=0,Z=0,W=1),Translation=(X=3,Y=0,Z=-2),Scale3D=(X=1,Y=1,Z=1))))
    Weight=@n10.OutputParam
   )
   > n11 = &RigUnit_SphericalPoseReader(
    ActiveRegionScaleFactors=(PositiveWidth=0.2,NegativeWidth=0.2,PositiveHeight=0.1,NegativeHeight=1)
    ActiveRegionSize=0.5
    Debug=(bDrawDebug=False,bDraw2D=false,bDrawLocalAxes=false,DebugScale=30,DebugSegments=30,DebugThickness=0.2)
    DriverAxis=(X=-1,Y=0,Z=0)
    DriverItem=(Type=Bone,Name="hand_r")
    FalloffRegionScaleFactors=(PositiveWidth=0.5,NegativeWidth=0.5,PositiveHeight=1,NegativeHeight=1)
    FalloffSize=0.5
    OptionalParentItem=(Type=Bone,Name="None")
    RotationOffset=(X=0,Y=0,Z=90)
   )
   > &RigUnit_ModifyTransforms(
    ItemToModify=((Item=(Type=Bone,Name="R_bracelet_r_side"),Transform=(Rotation=(X=0,Y=0,Z=0,W=1),Translation=(X=3,Y=-2,Z=0),Scale3D=(X=1,Y=1,Z=1))))
    Weight=@n11.OutputParam
   )
   > n12 = &RigUnit_SphericalPoseReader(
    ActiveRegionScaleFactors=(PositiveWidth=0.2,NegativeWidth=0.2,PositiveHeight=0.1,NegativeHeight=1)
    ActiveRegionSize=0.5
    Debug=(bDrawDebug=False,bDraw2D=false,bDrawLocalAxes=false,DebugScale=30,DebugSegments=30,DebugThickness=0.2)
    DriverAxis=(X=-1,Y=0,Z=0)
    DriverItem=(Type=Bone,Name="hand_r")
    FalloffRegionScaleFactors=(PositiveWidth=0.5,NegativeWidth=0.5,PositiveHeight=1,NegativeHeight=1)
    FalloffSize=0.5
    OptionalParentItem=(Type=Bone,Name="None")
    RotationOffset=(X=0,Y=180,Z=90)
   )
   > &RigUnit_ModifyTransforms(
    ItemToModify=((Item=(Type=Bone,Name="R_bracelet_l_side"),Transform=(Rotation=(X=0,Y=0,Z=0,W=1),Translation=(X=3,Y=2,Z=0),Scale3D=(X=1,Y=1,Z=1))))
    Weight=@n12.OutputParam
   )
   > n13 = &RigUnit_SphericalPoseReader(
    ActiveRegionScaleFactors=(PositiveWidth=0.2,NegativeWidth=0.2,PositiveHeight=1,NegativeHeight=0.1)
    ActiveRegionSize=0.3
    Debug=(bDrawDebug=False,bDraw2D=false,bDrawLocalAxes=false,DebugScale=30,DebugSegments=30,DebugThickness=0.2)
    DriverAxis=(X=-1,Y=0,Z=0)
    DriverItem=(Type=Bone,Name="hand_r")
    FalloffRegionScaleFactors=(PositiveWidth=0.5,NegativeWidth=0.5,PositiveHeight=1,NegativeHeight=1)
    FalloffSize=0.5
    OptionalParentItem=(Type=Bone,Name="None")
    RotationOffset=(X=0,Y=90,Z=-90)
   )
   > &RigUnit_ModifyTransforms(
    ItemToModify=((Item=(Type=Bone,Name="R_bracelet_lower"),Transform=(Rotation=(X=0,Y=0,Z=0,W=1),Translation=(X=3,Y=0,Z=2),Scale3D=(X=1,Y=1,Z=1))))
    Weight=@n13.OutputParam
   )
   n14 = &RigUnit_SphericalPoseReader(
    ActiveRegionScaleFactors=(PositiveWidth=0.8,NegativeWidth=0.8,PositiveHeight=0.8,NegativeHeight=0.8)
    ActiveRegionSize=0.2
    Debug=(bDrawDebug=False,bDraw2D=false,bDrawLocalAxes=false,DebugScale=30,DebugSegments=30,DebugThickness=0.2)
    DriverAxis=(X=-1,Y=0,Z=0)
    DriverItem=(Type=Bone,Name="calf_r")
    FalloffRegionScaleFactors=(PositiveWidth=0.5,NegativeWidth=0.5,PositiveHeight=0.5,NegativeHeight=0.5)
    FalloffSize=0.5
    OptionalParentItem=(Type=Bone,Name="None")
    RotationOffset=(X=-90,Y=0,Z=50)
    ExecutePin=@RigVMFunction_Sequence.E
   )
   n15 = &RigVMFunction_AlphaInterp(InterpSpeedDecreasing=10, InterpSpeedIncreasing=10, Scale=1, bClampResult=False, Value=@n14.OutputParam)
   > &RigUnit_ModifyTransforms(
    ItemToModify=((Item=(Type=Bone,Name="R_Extras02"),Transform=(Rotation=(X=0,Y=0,Z=0,W=1),Translation=(X=8,Y=0,Z=0),Scale3D=(X=1,Y=1,Z=1))),(Item=(Type=Bone,Name="R_bracelet"),Transform=(Rotation=(X=0,Y=0,Z=0,W=1),Translation=(X=0,Y=0,Z=0),Scale3D=(X=1,Y=1,Z=1))))
    Weight=@n15.Result
   )
   > &RigUnit_AimConstraintLocalSpaceOffset(
    AdvancedSettings=(DebugSettings=(bEnabled=False,Scale=10,WorldOffset=(Rotation=(X=0,Y=0,Z=0,W=1),Translation=(X=0,Y=0,Z=0),Scale3D=(X=1,Y=1,Z=1))),RotationOrderForFilter=XYZ)
    Child=(Type=Bone,Name="R_Extras01")
    Filter=(bX=false,bY=false,bZ=True)
    Parents=((Item=(Type=Bone,Name="R_Extras02"),Weight=1))
    UpAxis=(X=0,Y=1,Z=0)
    Weight=1
    WorldUp=(Target=(X=0,Y=0,Z=1),Kind=Location,Space=(Type=None,Name="None"))
    bMaintainOffset=true
   )
   > n16 = &RigUnit_SphericalPoseReader(
    ActiveRegionScaleFactors=(PositiveWidth=0.8,NegativeWidth=0.8,PositiveHeight=0.8,NegativeHeight=0.8)
    ActiveRegionSize=0.2
    Debug=(bDrawDebug=False,bDraw2D=false,bDrawLocalAxes=false,DebugScale=30,DebugSegments=30,DebugThickness=0.2)
    DriverAxis=(X=1,Y=0,Z=0)
    DriverItem=(Type=Bone,Name="calf_l")
    FalloffRegionScaleFactors=(PositiveWidth=0.5,NegativeWidth=0.5,PositiveHeight=0.5,NegativeHeight=0.5)
    FalloffSize=0.5
    OptionalParentItem=(Type=Bone,Name="None")
    RotationOffset=(X=90,Y=0,Z=50)
   )
   n17 = &RigVMFunction_AlphaInterp(InterpSpeedDecreasing=10, InterpSpeedIncreasing=10, Scale=1, bClampResult=False, Value=@n16.OutputParam)
   > &RigUnit_ModifyTransforms(
    ItemToModify=((Item=(Type=Bone,Name="L_Extras02"),Transform=(Rotation=(X=0,Y=0,Z=0,W=1),Translation=(X=-8,Y=0,Z=0),Scale3D=(X=1,Y=1,Z=1))),(Item=(Type=Bone,Name="R_bracelet"),Transform=(Rotation=(X=0,Y=0,Z=0,W=1),Translation=(X=0,Y=0,Z=0),Scale3D=(X=1,Y=1,Z=1))))
    Weight=@n17.Result
   )
   > &RigUnit_AimConstraintLocalSpaceOffset(
    AdvancedSettings=(DebugSettings=(bEnabled=False,Scale=10,WorldOffset=(Rotation=(X=0,Y=0,Z=0,W=1),Translation=(X=0,Y=0,Z=0),Scale3D=(X=1,Y=1,Z=1))),RotationOrderForFilter=XYZ)
    Child=(Type=Bone,Name="L_Extras01")
    Filter=(bX=false,bY=false,bZ=True)
    Parents=((Item=(Type=Bone,Name="L_Extras02"),Weight=1))
    UpAxis=(X=0,Y=1,Z=0)
    Weight=1
    WorldUp=(Target=(X=0,Y=0,Z=1),Kind=Location,Space=(Type=None,Name="None"))
    bMaintainOffset=true
   )
   &RigUnit_RotationConstraintLocalSpaceOffset(
    AdvancedSettings=(InterpolationType=Average,RotationOrderForFilter=XZY)
    Child=(Type=Bone,Name="L_elpow_armor")
    Filter=(bX=True,bY=True,bZ=True)
    Parents=((Item=(Type=Bone,Name="upperarm_l"),Weight=0.5),(Item=(Type=Bone,Name="lowerarm_l"),Weight=0.5))
    Weight=1
    ExecutePin=@RigVMFunction_Sequence.F
   )
   > &RigUnit_RotationConstraintLocalSpaceOffset(
    AdvancedSettings=(InterpolationType=Average,RotationOrderForFilter=XZY)
    Child=(Type=Bone,Name="R_elpow_armor")
    Filter=(bX=True,bY=True,bZ=True)
    Parents=((Item=(Type=Bone,Name="upperarm_r"),Weight=0.5),(Item=(Type=Bone,Name="lowerarm_r"),Weight=0.5))
    Weight=1
   )
   Get_Animation_Attribute_2 = RigVMDispatchNode(BoneName=root, Default=0, Name=deformRig_l_elbow_alpha) {template="DISPATCH_RigDispatch_GetAnimAttribute(in Name,in BoneName,in Default,out Value,out Found)"}
   > &RigUnit_ModifyTransforms(
    ItemToModify=((Item=(Type=Bone,Name="L_elpow_armor"),Transform=(Rotation=(X=0,Y=0,Z=0,W=1),Translation=(X=0,Y=1,Z=0),Scale3D=(X=1,Y=1,Z=1))),(Item=(Type=Bone,Name="R_elpow_armor"),Transform=(Rotation=(X=0,Y=0,Z=0,W=1),Translation=(X=0,Y=0,Z=0),Scale3D=(X=1,Y=1,Z=1))))
    Weight=@Get_Animation_Attribute_2.Value
   )
   Get_Animation_Attribute_1 = RigVMDispatchNode(BoneName=root, Default=0, Name=deformRig_r_elbow_alpha) {template="DISPATCH_RigDispatch_GetAnimAttribute(in Name,in BoneName,in Default,out Value,out Found)"}
   > &RigUnit_ModifyTransforms(
    ItemToModify=((Item=(Type=Bone,Name="R_elpow_armor"),Transform=(Rotation=(X=0,Y=0,Z=0,W=1),Translation=(X=0,Y=-1,Z=0),Scale3D=(X=1,Y=1,Z=1))))
    Weight=@Get_Animation_Attribute_1.Value
   )
   n18 = &RigUnit_SphericalPoseReader(
    ActiveRegionScaleFactors=(PositiveWidth=0.8,NegativeWidth=1,PositiveHeight=0.8,NegativeHeight=0.8)
    ActiveRegionSize=0.2
    Debug=(bDrawDebug=False,bDraw2D=false,bDrawLocalAxes=false,DebugScale=30,DebugSegments=30,DebugThickness=0.2)
    DriverAxis=(X=1,Y=0,Z=0)
    DriverItem=(Type=Bone,Name="foot_r")
    FalloffRegionScaleFactors=(PositiveWidth=0.5,NegativeWidth=1,PositiveHeight=0.5,NegativeHeight=0.5)
    FalloffSize=0.5
    OptionalParentItem=(Type=Bone,Name="None")
    RotationOffset=(X=-90,Y=0,Z=10)
    ExecutePin=@RigVMFunction_Sequence.G
   )
   > &RigUnit_ModifyTransforms(
    ItemToModify="((Item=(Type=Bone,Name=\"L_lower_leg_armor\"),Transform=(Rotation=(X=0,Y=0,Z=0,W=1),Translation=(X=3,Y=0,Z=0),Scale3D=(X=1,Y=1,Z=1))),(Item=(Type=Bone,Name=\"R_elpow_armor\"),Transform=(Rotation=(X=0,Y=0,Z=0,W=1),Translation=(X=0,Y=0,Z=0),Scale3D=(X=1,Y=1,Z=1) ...(truncated)"
    Weight=@n18.OutputParam
   )
   > n19 = &RigUnit_SphericalPoseReader(
    ActiveRegionScaleFactors=(PositiveWidth=1,NegativeWidth=0.8,PositiveHeight=0.8,NegativeHeight=0.8)
    ActiveRegionSize=0.2
    Debug=(bDrawDebug=False,bDraw2D=false,bDrawLocalAxes=false,DebugScale=30,DebugSegments=30,DebugThickness=0.2)
    DriverAxis=(X=-1,Y=0,Z=0)
    DriverItem=(Type=Bone,Name="foot_l")
    FalloffRegionScaleFactors=(PositiveWidth=1,NegativeWidth=0.5,PositiveHeight=0.5,NegativeHeight=0.5)
    FalloffSize=0.5
    OptionalParentItem=(Type=Bone,Name="None")
    RotationOffset=(X=90,Y=0,Z=10)
   )
   > &RigUnit_ModifyTransforms(
    ItemToModify="((Item=(Type=Bone,Name=\"R_lower_leg_armor\"),Transform=(Rotation=(X=0,Y=0,Z=0,W=1),Translation=(X=-3,Y=0,Z=0),Scale3D=(X=1,Y=1,Z=1))),(Item=(Type=Bone,Name=\"R_Extras03\"),Transform=(Rotation=(X=0,Y=0,Z=0,W=1),Translation=(X=0,Y=0,Z=0),Scale3D=(X=1,Y=1,Z=1))) ...(truncated)"
    Weight=@n19.OutputParam
   )
   &RigUnit_RotationConstraintLocalSpaceOffset(
    AdvancedSettings=(InterpolationType=Average,RotationOrderForFilter=XZY)
    Child=(Type=Bone,Name="L_leg_toe")
    Filter=(bX=True,bY=True,bZ=True)
    Parents=((Item=(Type=Bone,Name="ball_l"),Weight=1))
    Weight=1
    ExecutePin=@RigVMFunction_Sequence.B
   )
   > &RigUnit_RotationConstraintLocalSpaceOffset(
    AdvancedSettings=(InterpolationType=Average,RotationOrderForFilter=XZY)
    Child=(Type=Bone,Name="R_leg_toe")
    Filter=(bX=True,bY=True,bZ=True)
    Parents=((Item=(Type=Bone,Name="ball_r"),Weight=1))
    Weight=1
   )
   > n20 = &RigUnit_SphericalPoseReader(
    ActiveRegionScaleFactors=(PositiveWidth=1,NegativeWidth=0.1,PositiveHeight=1,NegativeHeight=0.5)
    ActiveRegionSize=0.3
    Debug=(bDrawDebug=False,bDraw2D=false,bDrawLocalAxes=false,DebugScale=30,DebugSegments=30,DebugThickness=0.2)
    DriverAxis=(X=1,Y=0,Z=0)
    DriverItem=(Type=Bone,Name="clavicle_l")
    FalloffRegionScaleFactors=(PositiveWidth=1,NegativeWidth=0.5,PositiveHeight=1,NegativeHeight=0.5)
    FalloffSize=0.5
    OptionalParentItem=(Type=Bone,Name="None")
    RotationOffset=(X=0,Y=40,Z=180)
   )
   AlphaInterp = &RigVMFunction_AlphaInterp(InterpSpeedDecreasing=10, InterpSpeedIncreasing=10, Scale=1, bClampResult=False, Value=@n20.OutputParam)
   > &RigUnit_ModifyTransforms(
    ItemToModify="((Item=(Type=Bone,Name=\"BASE_L_collar_01\"),Transform=(Rotation=(X=0,Y=0,Z=0,W=1),Translation=(X=0,Y=0,Z=0),Scale3D=(X=1,Y=1,Z=1))),(Item=(Type=Bone,Name=\"BASE_L_collar_02\"),Transform=(Rotation=(X=0,Y=0,Z=0,W=1),Translation=(X=0,Y=1,Z=0),Scale3D=(X=1,Y=1,Z= ...(truncated)"
    Weight=@AlphaInterp.Result
    ItemToModify.0.Transform.Rotation=&RigVMFunction_MathQuaternionFromEuler(Euler=(X=20,Y=0,Z=-170), RotationOrder=XYZ).Result
    ItemToModify.1.Transform.Rotation=&RigVMFunction_MathQuaternionFromEuler(Euler=(X=0,Y=0,Z=-155), RotationOrder=XYZ).Result
    ItemToModify.2.Transform.Rotation=&RigVMFunction_MathQuaternionFromEuler(Euler=(X=0,Y=0,Z=-150), RotationOrder=XYZ).Result
    ItemToModify.3.Transform.Rotation=&RigVMFunction_MathQuaternionFromEuler(Euler=(X=0,Y=0,Z=-50), RotationOrder=XYZ).Result
   )
   > n21 = &RigUnit_SphericalPoseReader(
    ActiveRegionScaleFactors=(PositiveWidth=1,NegativeWidth=0.1,PositiveHeight=0.5,NegativeHeight=1)
    ActiveRegionSize=0.3
    Debug=(bDrawDebug=False,bDraw2D=false,bDrawLocalAxes=false,DebugScale=30,DebugSegments=30,DebugThickness=0.2)
    DriverAxis=(X=-1,Y=0,Z=0)
    DriverItem=(Type=Bone,Name="clavicle_r")
    FalloffRegionScaleFactors=(PositiveWidth=1,NegativeWidth=0.5,PositiveHeight=1,NegativeHeight=1)
    FalloffSize=0.5
    OptionalParentItem=(Type=Bone,Name="None")
    RotationOffset=(X=0,Y=220,Z=180)
   )
   RigVMFunction_AnimEvalRichCurve = &RigVMFunction_AnimEvalRichCurve(
    Curve="(EditorCurveData=(Keys=((InterpMode=RCIM_Linear,TangentMode=RCTM_Auto,TangentWeightMode=RCTWM_WeightedNone,Time=0,Value=0,ArriveTangent=0,ArriveTangentWeight=0,LeaveTangent=0,LeaveTangentWeight=0),(InterpMode=RCIM_Linear,TangentMode=RCTM_Auto,TangentWeight ...(truncated)"
    SourceMinimum=0
    TargetMinimum=0
    Value=@n21.OutputParam
   )
   n22 = &RigVMFunction_AlphaInterp(InterpSpeedDecreasing=10, InterpSpeedIncreasing=10, Scale=1, bClampResult=False, Value=@RigVMFunction_AnimEvalRichCurve.Result)
   > &RigUnit_ModifyTransforms(
    ItemToModify="((Item=(Type=Bone,Name=\"BASE_R_collar_01\"),Transform=(Rotation=(X=0,Y=0,Z=0,W=1),Translation=(X=0,Y=0,Z=0),Scale3D=(X=1,Y=1,Z=1))),(Item=(Type=Bone,Name=\"BASE_R_collar_02\"),Transform=(Rotation=(X=0,Y=0,Z=0,W=1),Translation=(X=-1,Y=0,Z=0),Scale3D=(X=1,Y=1,Z ...(truncated)"
    Weight=@n22.Result
    ItemToModify.2.Transform.Rotation=&RigVMFunction_MathQuaternionFromEuler(Euler=(X=0,Y=0,Z=-150), RotationOrder=XYZ).Result
    ItemToModify.3.Transform.Rotation=&RigVMFunction_MathQuaternionFromEuler(Euler=(X=0,Y=0,Z=-50), RotationOrder=XYZ).Result
    ItemToModify.0.Transform.Rotation=&RigVMFunction_MathQuaternionFromEuler(Euler=(X=0,Y=0,Z=-100), RotationOrder=XYZ).Result
    ItemToModify.1.Transform.Rotation=&RigVMFunction_MathQuaternionFromEuler(Euler=(X=0,Y=0,Z=-120), RotationOrder=XYZ).Result
   )
   > &RigUnit_ModifyTransforms(
    ItemToModify=((Item=(Type=Bone,Name="R_collar_01"),Transform=(Rotation=(X=0,Y=0,Z=0,W=1),Translation=(X=0,Y=-3,Z=0),Scale3D=(X=1,Y=1,Z=1))))
    Weight=&RigVMFunction_MathDoubleMul(A=@n22.Result, B=RigVMVariableNode() {variable=skydiving}.Value).Result
    ItemToModify.0.Transform.Rotation=&RigVMFunction_MathQuaternionFromEuler(Euler=(X=0,Y=0,Z=10), RotationOrder=XYZ).Result
   )
  ]}
]
hierarchy [
 root {loc=<0 0 0>, rot=<0 0 0 1>}
 pelvis {parent=root, loc=<0 -0.7321 92.4218>, rot=<0.0127 0.707 -0.0127 -0.707>}
 spine_01 {parent=pelvis, loc=<5.6002 -0.5128 0>, rot=<0 0 0.034 -0.9994>}
 spine_02 {parent=spine_01, loc=<8.0756 0.0139 0>, rot=<0 0 0.0238 -0.9997>}
 spine_03 {parent=spine_02, loc=<7.0395 0 0>, rot=<0 0 0.0149 -0.9999>}
 spine_04 {parent=spine_03, loc=<8.1023 0 0>, rot=<0 0 -0.0132 -0.9999>}
 spine_05 {parent=spine_04, loc=<10.0141 0 0>, rot=<0 0 -0.0015 -1>}
 attach_backpack {parent=spine_05, loc=<-0.1227 -15.1426 0>, rot=<-0.0093 -0.707 0.0093 -0.707>}
 clavicle_l {parent=spine_05, loc=<2.5856 -0.2866 -4.3777>, rot=<0.0465 -0.7289 -0.009 -0.6829>}
 upperarm_l {parent=clavicle_l, loc=<18.2264 0.1342 -0.001>, rot=<0.0253 -0.2693 0.0013 -0.9627>}
 lowerarm_l {parent=upperarm_l, loc=<29.0071 0.1686 -1.239>, rot=<0.0391 -0.0099 -0.2453 -0.9686>}
 hand_l {parent=lowerarm_l, loc=<25.8607 0.222 1.0274>, rot=<-0.7193 -0.0988 0.1193 -0.6773>}
 thumb_01_l {parent=hand_l, loc=<1.3118 -0.3191 -2.459>, rot=<0.6546 -0.4042 0.0664 -0.6354>}
 thumb_02_l {parent=thumb_01_l, loc=<5.9336 -0.0448 0.0401>, rot=<-0.0037 0.0213 0.298 -0.9543>}
 thumb_03_l {parent=thumb_02_l, loc=<2.2223 0.3682 -0.0907>, rot=<0.0118 -0.013 0.0066 -0.9998>}
 index_metacarpal_l {parent=hand_l, loc=<1.6535 0.0183 -2.96>, rot=<0.0158 -0.0124 -0.0173 -0.9996>}
 index_01_l {parent=index_metacarpal_l, loc=<8.3008 0.0357 -0.0524>, rot=<-0.0044 0.0485 0.0924 -0.9945>}
 index_02_l {parent=index_01_l, loc=<3.3587 -0.3004 0.0452>, rot=<-0.002 0.0027 0.3225 -0.9466>}
 index_03_l {parent=index_02_l, loc=<2.5578 -0.123 -0.0252>, rot=<-0.002 0.0067 0.2558 -0.9667>}
 middle_metacarpal_l {parent=hand_l, loc=<2.1856 -0.1725 -0.5984>, rot=<0.006 0.0168 -0.0329 -0.9993>}
 middle_01_l {parent=middle_metacarpal_l, loc=<7.3275 -0.2447 0.0129>, rot=<-0.0208 0.0042 0.0858 -0.9961>}
 middle_02_l {parent=middle_01_l, loc=<4.0123 0.1401 0.1175>, rot=<-0.0218 0.0258 0.3512 -0.9357>}
 middle_03_l {parent=middle_02_l, loc=<2.6077 0.0534 0.1373>, rot=<0.0082 0.0188 0.1708 -0.9851>}
 ring_metacarpal_l {parent=hand_l, loc=<2.2074 -0.3625 1.3638>, rot=<-0.0284 0.0722 -0.0458 -0.9959>}
 ring_01_l {parent=ring_metacarpal_l, loc=<7.0295 -0.6197 -0.1446>, rot=<-0.0489 -0.0252 0.1892 -0.9804>}
 ring_02_l {parent=ring_01_l, loc=<3.6174 0.4647 -0.1027>, rot=<0 0 0.2806 -0.9598>}
 ring_03_l {parent=ring_02_l, loc=<2.4152 0.114 0.0199>, rot=<0 0 0.1189 -0.9929>}
 pinky_metacarpal_l {parent=hand_l, loc=<1.7616 -0.9753 3.3814>, rot=<-0.0397 0.0814 -0.069 -0.9935>}
 pinky_01_l {parent=pinky_metacarpal_l, loc=<7.1709 -0.4831 -0.1094>, rot=<-0.0654 -0.0035 0.2242 -0.9723>}
 pinky_02_l {parent=pinky_01_l, loc=<2.8274 0.0893 -0.1043>, rot=<0.0037 0.0011 0.2854 -0.9584>}
 pinky_03_l {parent=pinky_02_l, loc=<2.2046 0.0307 -0.0078>, rot=<-0.0002 0.0098 0.1445 -0.9895>}
 weapon_l {parent=hand_l, loc=<-0.1974 -3.4111 0.3811>, rot=<0 0 0 1>}
 hand_attach_l {parent=hand_l, loc=<8.5 -3 0>, rot=<0.7071 0.7071 0 0>}
 lowerarm_twist_01_l {parent=lowerarm_l, loc=<17.8549 0 0>, rot=<0 0 0 -1>}
 L_bracelet {parent=lowerarm_l, loc=<23.0578 0.4666 0.5534>, rot=<0 0 0 -1>}
 BASE_L_bracelet_upper {parent=L_bracelet, loc=<2.617 1.0444 5.3306>, rot=<0 0 0 1>}
 L_bracelet_upper {parent=BASE_L_bracelet_upper, loc=<0 0 0>, rot=<0 0 0 1>}
 BASE_L_bracelet_r_side {parent=L_bracelet, loc=<0 5.4372 0>, rot=<0 0 0 1>}
 L_bracelet_r_side {parent=BASE_L_bracelet_r_side, loc=<0 0 0>, rot=<0 0 0 1>}
 BASE_L_bracelet_l_side {parent=L_bracelet, loc=<0 -4.9847 0>, rot=<0 0 0 1>}
 L_bracelet_l_side______ {parent=BASE_L_bracelet_l_side, loc=<0 0 0>, rot=<0 0 0 1>}
 L_bracelet_l_side {parent=BASE_L_bracelet_l_side, loc=<0 0 0>, rot=<0 0 0 1>}
 BASE_L_bracelet_lower {parent=L_bracelet, loc=<-1.5596 0.1834 -3.973>, rot=<0 0 0 1>}
 L_bracelet_lower {parent=BASE_L_bracelet_lower, loc=<0 0 0>, rot=<0 0 0 1>}
 deform_elbow_out_base_l {parent=lowerarm_l, loc=<0 -7.5129 0>, rot=<0 0 0 1>}
 deform_elbow_out_l {parent=deform_elbow_out_base_l, loc=<0 0 0>, rot=<0 0 0 1>}
 L_elpow_armor {parent=lowerarm_l, loc=<0 0 0>, rot=<0 0 0 1>}
 L_elbow_armor {parent=lowerarm_l, loc=<0 0 0>, rot=<0 0 0 1>}
 upperarm_twist_01_l {parent=upperarm_l, loc=<7.9796 0.0843 -0.6195>, rot=<0 0 0 -1>}
 upperarm_twist_02_l {parent=upperarm_l, loc=<19.5059 0.0843 -0.6195>, rot=<0 0 0 -1>}
 neck_01 {parent=spine_05, loc=<10.815 1.9196 0>, rot=<0 0 -0.1864 -0.9825>}
 neck_02 {parent=neck_01, loc=<4.7347 0.3045 0>, rot=<0 0 0.0363 -0.9993>}
 head {parent=neck_02, loc=<5.7632 0.5552 0>, rot=<0 0 0.115 -0.9934>}
 attach_cape {parent=spine_05, loc=<9.1516 -13.3787 0>, rot=<0.0616 -0.7044 -0.0616 -0.7044>}
 clavicle_r {parent=spine_05, loc=<2.5855 -0.2866 4.3777>, rot=<0.7289 0.0465 -0.6829 0.009>}
 upperarm_r {parent=clavicle_r, loc=<-18.2264 -0.1342 0.001>, rot=<0.0253 -0.2693 0.0013 -0.9627>}
 lowerarm_r {parent=upperarm_r, loc=<-29.0071 -0.1686 1.239>, rot=<0.0391 -0.0099 -0.2453 -0.9686>}
 hand_r {parent=lowerarm_r, loc=<-25.8607 -0.222 -1.0274>, rot=<-0.7193 -0.0988 0.1193 -0.6773>}
 thumb_01_r {parent=hand_r, loc=<-1.3118 0.3191 2.459>, rot=<0.6546 -0.4042 0.0664 -0.6354>}
 thumb_02_r {parent=thumb_01_r, loc=<-5.9336 0.0448 -0.0401>, rot=<-0.0037 0.0213 0.298 -0.9543>}
 thumb_03_r {parent=thumb_02_r, loc=<-2.2223 -0.3682 0.0907>, rot=<0.0118 -0.013 0.0066 -0.9998>}
 index_metacarpal_r {parent=hand_r, loc=<-1.6535 -0.0183 2.96>, rot=<0.0158 -0.0124 -0.0173 -0.9996>}
 index_01_r {parent=index_metacarpal_r, loc=<-8.3008 -0.0357 0.0524>, rot=<-0.0044 0.0485 0.0924 -0.9945>}
 index_02_r {parent=index_01_r, loc=<-3.3587 0.3004 -0.0452>, rot=<-0.002 0.0027 0.3225 -0.9466>}
 index_03_r {parent=index_02_r, loc=<-2.5578 0.123 0.0252>, rot=<-0.002 0.0067 0.2558 -0.9667>}
 middle_metacarpal_r {parent=hand_r, loc=<-2.1856 0.1725 0.5984>, rot=<0.006 0.0168 -0.0329 -0.9993>}
 middle_01_r {parent=middle_metacarpal_r, loc=<-7.3275 0.2447 -0.0129>, rot=<-0.0208 0.0042 0.0858 -0.9961>}
 middle_02_r {parent=middle_01_r, loc=<-4.0123 -0.1401 -0.1175>, rot=<-0.0218 0.0258 0.3512 -0.9357>}
 middle_03_r {parent=middle_02_r, loc=<-2.6077 -0.0534 -0.1373>, rot=<0.0082 0.0188 0.1708 -0.9851>}
 ring_metacarpal_r {parent=hand_r, loc=<-2.2074 0.3625 -1.3638>, rot=<-0.0284 0.0722 -0.0458 -0.9959>}
 ring_01_r {parent=ring_metacarpal_r, loc=<-7.0295 0.6197 0.1446>, rot=<-0.0489 -0.0252 0.1892 -0.9804>}
 ring_02_r {parent=ring_01_r, loc=<-3.6174 -0.4647 0.1027>, rot=<0 0 0.2806 -0.9598>}
 ring_03_r {parent=ring_02_r, loc=<-2.4152 -0.114 -0.0199>, rot=<0 0 0.1189 -0.9929>}
 pinky_metacarpal_r {parent=hand_r, loc=<-1.7616 0.9753 -3.3814>, rot=<-0.0397 0.0814 -0.069 -0.9935>}
 pinky_01_r {parent=pinky_metacarpal_r, loc=<-7.1709 0.4831 0.1094>, rot=<-0.0654 -0.0035 0.2242 -0.9723>}
 pinky_02_r {parent=pinky_01_r, loc=<-2.8274 -0.0893 0.1043>, rot=<0.0037 0.0011 0.2854 -0.9584>}
 pinky_03_r {parent=pinky_02_r, loc=<-2.2046 -0.0307 0.0078>, rot=<-0.0002 0.0098 0.1445 -0.9895>}
 hand_attach_r {parent=hand_r, loc=<-8.5 3 0>, rot=<0 0 -0.7071 -0.7071>}
 weapon_r {parent=hand_r, loc=<0.1974 3.4111 -0.3811>, rot=<0 0 0 1>}
 lowerarm_twist_01_r {parent=lowerarm_r, loc=<-17.8549 0 0>, rot=<0 0 0 -1>}
 R_bracelet {parent=lowerarm_r, loc=<-23.0578 -0.4666 -0.5534>, rot=<0 0 0 -1>}
 BASE_R_bracelet_upper {parent=R_bracelet, loc=<-2.617 -1.0444 -5.3305>, rot=<0 0 0 -1>}
 R_bracelet_upper {parent=BASE_R_bracelet_upper, loc=<0 0 0>, rot=<0 0 0 1>}
 BASE_R_bracelet_r_side {parent=R_bracelet, loc=<-0.0002 -5.4371 0.0004>, rot=<0 0 0 -1>}
 R_bracelet_r_side {parent=BASE_R_bracelet_r_side, loc=<0 0 0>, rot=<0 0 0 1>}
 BASE_R_bracelet_lower {parent=R_bracelet, loc=<1.5595 -0.1834 3.9731>, rot=<0 0 0 -1>}
 R_bracelet_lower {parent=BASE_R_bracelet_lower, loc=<0 0 0>, rot=<0 0 0 1>}
 BASE_R_bracelet_l_side {parent=R_bracelet, loc=<-0.0001 4.9847 0.0002>, rot=<0 0 0 -1>}
 R_bracelet_l_side______ {parent=BASE_R_bracelet_l_side, loc=<0 0 0>, rot=<0 0 0 1>}
 R_bracelet_l_side {parent=BASE_R_bracelet_l_side, loc=<0 0 0>, rot=<0 0 0 1>}
 deform_elbow_out_base_r {parent=lowerarm_r, loc=<-0.0001 7.5129 0.0002>, rot=<0 0 0 1>}
 deform_elbow_out_r {parent=deform_elbow_out_base_r, loc=<0 0 0>, rot=<0 0 0 1>}
 R_elpow_armor {parent=lowerarm_r, loc=<0 0 0>, rot=<0 0 0 1>}
 R_elbow_armor {parent=lowerarm_r, loc=<0 0 0>, rot=<0 0 0 1>}
 upperarm_twist_01_r {parent=upperarm_r, loc=<-7.9796 -0.0843 0.6195>, rot=<0 0 0 -1>}
 deform_upperarm_fr_base_r {parent=upperarm_twist_01_r, loc=<8.3485 -10.7117 2.1456>, rot=<0 0 0 -1>}
 deform_upperarm_fr_r {parent=deform_upperarm_fr_base_r, loc=<0 0 0>, rot=<0 0 0 1>}
 upperarm_twist_02_r {parent=upperarm_r, loc=<-19.5059 -0.0843 0.6195>, rot=<0 0 0 -1>}
 BASE_L_collar_01 {parent=spine_05, loc=<12.7478 10.4049 -15.1573>, rot=<-0.1161 0.6372 0.7488 0.141>}
 L_collar_01 {parent=BASE_L_collar_01, loc=<0 0 0>, rot=<0 0 0 1>}
 L_collar_01_extra {parent=L_collar_01, loc=<0 0 0>, rot=<0 0 0 1>}
 BASE_L_collar_02 {parent=spine_05, loc=<14.7714 3.6278 -16.2246>, rot=<-0.2361 0.6805 0.6814 0.1298>}
 L_collar_02 {parent=BASE_L_collar_02, loc=<0 0 0>, rot=<0 0 0 1>}
 BASE_L_collar_03 {parent=spine_05, loc=<16.1038 -2.4393 -14.7875>, rot=<0.1815 -0.8543 -0.4811 -0.0758>}
 L_collar_03 {parent=BASE_L_collar_03, loc=<0 0 0>, rot=<0 0 0 1>}
 L_collar_03_extra {parent=L_collar_03, loc=<0 0 0>, rot=<0 0 0 1>}
 BASE_L_collar_04 {parent=spine_05, loc=<17.0631 -8.3213 -10.4151>, rot=<0.0405 -0.9536 -0.296 -0.0361>}
 L_collar_04 {parent=BASE_L_collar_04, loc=<0 0 0>, rot=<0 0 0 1>}
 BASE_R_collar_04 {parent=spine_05, loc=<17.4529 -9.7917 8.1904>, rot=<0.1059 -0.9679 0.2076 0.0937>}
 R_collar_04 {parent=BASE_R_collar_04, loc=<0 0 0>, rot=<0 0 0 1>}
 BASE_R_collar_03 {parent=spine_05, loc=<16.6648 -3.9174 13.7446>, rot=<0.2496 -0.8511 0.4223 0.1872>}
 R_collar_03 {parent=BASE_R_collar_03, loc=<0 0 0>, rot=<0 0 0 1>}
 BASE_R_collar_02 {parent=spine_05, loc=<15.298 1.0526 15.6203>, rot=<0.2916 -0.6941 0.6043 0.2609>}
 R_collar_02 {parent=BASE_R_collar_02, loc=<0 0 0>, rot=<0 0 0 1>}
 BASE_R_collar_01 {parent=spine_05, loc=<13.4473 5.7921 15.7531>, rot=<0.3372 -0.5991 0.6381 0.3468>}
 R_collar_01 {parent=BASE_R_collar_01, loc=<0 0 0>, rot=<0 0 0 1>}
 BASE_L_front_jacket_03 {parent=spine_04, loc=<0.7675 9.6176 -13.0308>, rot=<0.0595 -0.9387 -0.3382 0.0301>}
 L_front_jacket_03 {parent=BASE_L_front_jacket_03, loc=<0 0 0>, rot=<0 0 0 1>}
 L_front_jacket_02 {parent=L_front_jacket_03, loc=<6.2757 0 0>, rot=<0 0 0 1>}
 zipper {parent=L_front_jacket_02, loc=<5.9033 4.1078 -0.0933>, rot=<0.14 0.0179 -0.0065 -0.99>}
 BASE_L_back_jacket_03 {parent=spine_04, loc=<5.9391 -12.6003 -7.0509>, rot=<0.0033 -1 -0.0005 -0.0067>}
 L_back_jacket_03 {parent=BASE_L_back_jacket_03, loc=<0 0 0>, rot=<0 0 0 1>}
 L_back_jacket_02 {parent=L_back_jacket_03, loc=<6.5954 0 0>, rot=<0 0 0 1>}
 BASE_R_back_jacket_03 {parent=spine_04, loc=<5.8231 -12.604 6.3375>, rot=<0.0033 -1 -0.0005 -0.0067>}
 R_back_jacket_03 {parent=BASE_R_back_jacket_03, loc=<0 0 0>, rot=<0 0 0 1>}
 R_back_jacket_02 {parent=R_back_jacket_03, loc=<6.5954 0 0>, rot=<0 0 0 1>}
 BASE_R_side_jacket_02 {parent=spine_04, loc=<-2.6153 -1.7008 16.1804>, rot=<0.0029 -0.0701 0.9967 -0.0415>}
 R_side_jacket_02 {parent=BASE_R_side_jacket_02, loc=<0 0 0>, rot=<0 0 0 1>}
 BASE_R_front_jacket_03 {parent=spine_04, loc=<0.9302 8.9749 13.1953>, rot=<-0.0277 0.4041 0.9139 -0.0275>}
 R_front_jacket_03 {parent=BASE_R_front_jacket_03, loc=<0 0 0>, rot=<0 0 0 1>}
 R_front_jacket_02 {parent=R_front_jacket_03, loc=<6.2757 0 0>, rot=<0 0 0 1>}
 BASE_L_side_jacket_02 {parent=spine_04, loc=<-2.1137 -3.7114 -15.9235>, rot=<0.0416 -0.9983 0.0398 -0.0017>}
 L_side_jacket_02 {parent=BASE_L_side_jacket_02, loc=<0 0 0>, rot=<0 0 0 1>}
 pelvis_mech_turning {parent=spine_01, loc=<4.5151 -0.4627 0>, rot=<0 0 0 1>}
 pelvis_mech_slant {parent=pelvis_mech_turning, loc=<0 0 0>, rot=<0 0 0 1>}
 wheel {parent=pelvis_mech_slant, loc=<8.1301 0.2833 0.0042>, rot=<0 -0.0002 -0.0027 -1>}
 upper_body {parent=wheel, loc=<0 0 0>, rot=<0 0.0001 0 -1>}
 joint8 {parent=wheel, loc=<0.0098 0.0708 -6.8176>, rot=<-0.0052 0.0291 0.1492 -0.9884>}
 joint10 {parent=wheel, loc=<0.0094 0.0708 6.8176>, rot=<0.0291 0.0052 0.9884 0.1492>}
 joint9 {parent=pelvis_mech_turning, loc=<0.3842 2.7075 7.2717>, rot=<0.0291 0.0053 0.9888 0.1465>}
 joint7 {parent=pelvis_mech_turning, loc=<0.3847 2.7075 -7.2717>, rot=<-0.0053 0.0291 0.1465 -0.9888>}
 thigh_l {parent=pelvis, loc=<-1.8716 -0.0109 -11.9936>, rot=<0.0012 0.0572 0.0155 -0.9982>}
 calf_l {parent=thigh_l, loc=<-41.9178 -0.2246 0.0926>, rot=<0.0006 -0.0161 -0.0391 -0.9991>}
 foot_l {parent=calf_l, loc=<-38.7057 1.2438 0.2452>, rot=<0.0975 -0.0474 0.0317 -0.9936>}
 ball_l {parent=foot_l, loc=<-6.1614 13.1469 0.6807>, rot=<-0.0013 -0.0013 -0.708 -0.7062>}
 L_leg_toe {parent=foot_l, loc=<-8.2054 5.3544 1.561>, rot=<-0.618 -0.7831 -0.0563 0.0412>}
 calf_twist_02_l {parent=calf_l, loc=<-15.8677 -0.1066 -0.0084>, rot=<0.0535 0.0088 -0.0029 -0.9985>}
 calf_twist_01_l {parent=calf_l, loc=<-32.8048 -0.1321 0.0265>, rot=<0.1153 0.0065 -0.006 -0.9933>}
 L_BASE_lower_leg_armor {parent=calf_l, loc=<-30.2401 -2.5151 1.3456>, rot=<-0.0017 0.0022 0.2034 -0.9791>}
 L_lower_leg_armor {parent=L_BASE_lower_leg_armor, loc=<0 0 0>, rot=<0 0 0 1>}
 L_Extras01 {parent=calf_l, loc=<0 0 0>, rot=<0.8728 0.4851 -0.0507 0.0189>}
 L_BASE_Extras02 {parent=calf_l, loc=<-7.5352 -12.1465 1.5346>, rot=<0.981 -0.1941 0.0045 -0.0019>}
 L_Extras02 {parent=L_BASE_Extras02, loc=<0 0 0>, rot=<0 0 0 1>}
 thigh_r {parent=pelvis, loc=<-1.8716 -0.0109 11.9936>, rot=<0.0572 -0.0012 0.9982 0.0155>}
 calf_r {parent=thigh_r, loc=<41.9178 0.2246 -0.0926>, rot=<0.0006 -0.0161 -0.0391 -0.9991>}
 foot_r {parent=calf_r, loc=<38.7057 -1.2438 -0.2452>, rot=<0.0975 -0.0474 0.0317 -0.9936>}
 ball_r {parent=foot_r, loc=<6.1616 -13.1473 -0.6805>, rot=<0.0001 -0.0001 -0.7065 -0.7077>}
 R_leg_toe {parent=foot_r, loc=<8.2054 -5.3544 -1.561>, rot=<-0.618 -0.7831 -0.0563 0.0412>}
 calf_twist_02_r {parent=calf_r, loc=<15.8691 0.1203 0.0276>, rot=<-0.0964 0.0077 -0.0043 -0.9953>}
 calf_twist_01_r {parent=calf_r, loc=<32.8075 0.1472 -0.0196>, rot=<-0.0146 0.0141 0.0025 -0.9998>}
 R_BASE_lower_leg_armor {parent=calf_r, loc=<30.2401 2.5151 -1.3456>, rot=<-0.0017 0.0022 0.2034 -0.9791>}
 R_lower_leg_armor {parent=R_BASE_lower_leg_armor, loc=<0 0 0>, rot=<0 0 0 1>}
 joint13 {parent=calf_r, loc=<7.5352 12.1465 -1.5346>, rot=<0.981 -0.1941 0.0045 -0.0019>}
 joint11 {parent=joint13, loc=<0 0 0>, rot=<0 0 0 1>}
 joint14 {parent=calf_r, loc=<0 0 0>, rot=<0.8704 0.4779 -0.0978 -0.0664>}
 R_Extras01 {parent=calf_r, loc=<0 0 0>, rot=<0.8728 0.4851 -0.0507 0.0189>}
 R_Extras03 {parent=R_Extras01, loc=<13.5447 0.974 6.4821>, rot=<-0.0279 -0.0509 -0.6453 -0.7617>}
 R_Extras04 {parent=R_Extras01, loc=<12.8405 0.7757 -6.8608>, rot=<-0.0073 -0.0199 -0.6435 -0.7652>}
 R_BASE_Extras02 {parent=calf_r, loc=<7.5352 12.1465 -1.5346>, rot=<0.981 -0.1941 0.0045 -0.0019>}
 R_Extras02 {parent=R_BASE_Extras02, loc=<0 0 0>, rot=<0 0 0 1>}
]
variables [
 ItemToModify_0_Transform_Rotation {type=FQuat, default=(X=0,Y=0,Z=0,W=1)}
 skydiving {type=double, default=0}
 static_crouching {type=double, default=0}
]
