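# Physics asset for the large male BadCat player body, pairing coarse body collision capsules with jointed chains for its jacket.
# Capsules cover pelvis, spine_04, arms and legs, linked by uniformly limited constraints, with no head body.
# Six jacket chains (left and right front, side and back) hang from spine_04 through tiny BASE anchor capsules with free angular motion.
# Segments below each anchor have tight swing and twist limits, stiffest at the back.
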PhysicsAsset /BRCosmetics/Characters/Player/Male/Large/Bodies/M_LRG_BadCat/Meshes/M_LRG_BadCat_Physics
bodies {
 BASE_L_back_jacket_03 [{shapeType=Capsule, length=0.505, radius=0.505, shapeName=BASE_L_back_jacket_03_capsule, rotation=<rot 0 -90 0>}]
 BASE_L_front_jacket_03 [{shapeType=Capsule, length=0.505, radius=0.505, shapeName=BASE_L_front_jacket_03_capsule, rotation=<rot 0 -90 0>}]
 BASE_L_side_jacket_02 [{shapeType=Capsule, length=0.505, radius=0.505, shapeName=BASE_L_side_jacket_02_capsule, rotation=<rot 0 -90 0>}]
 BASE_R_back_jacket_03 [{shapeType=Capsule, length=0.505, radius=0.505, shapeName=BASE_R_back_jacket_03_capsule, rotation=<rot 0 -90 0>}]
 BASE_R_front_jacket_03 [{shapeType=Capsule, length=0.505, radius=0.505, shapeName=BASE_R_front_jacket_03_capsule, rotation=<rot 0 -90 0>}]
 BASE_R_side_jacket_02 [{shapeType=Capsule, length=0.505, radius=0.505, shapeName=BASE_R_side_jacket_02_capsule, rotation=<rot 0 -90 0>}]
 L_back_jacket_02 [{
   shapeType=Capsule
   length=9.008
   radius=3.5568
   shapeName=L_back_jacket_02_capsule
   center=<4.3068 -0.4546 -0.3233>
   rotation=<rot 11.6243 -90 5.4279>
  }]
 L_back_jacket_03 [{
   shapeType=Capsule
   length=4.9957
   radius=4.2861
   shapeName=L_back_jacket_03_capsule
   center=<4.4312 -0.3417 -0.0782>
   rotation=<rot 9.938 -90 4.6539>
  }]
 L_front_jacket_02 [{
   shapeType=Capsule
   length=7.7066
   radius=3.465
   shapeName=L_front_jacket_02_capsule
   center=<3.872 -2.5487 -0.2128>
   rotation=<rot -1.1937 178.5373 -77.1574>
  }]
 L_front_jacket_03 [{
   shapeType=Capsule
   length=6.2822
   radius=3.612
   shapeName=L_front_jacket_03_capsule
   center=<3.5175 -2.3324 0.7442>
   rotation=<rot 9.2485 -176.1117 -72.0426>
  }]
 L_side_jacket_02 [{
   shapeType=Capsule
   length=6.7592
   radius=4.1567
   shapeName=L_side_jacket_02_capsule
   center=<2.9168 0.0409 0.0421>
   rotation=<rot -0.0323 170.3415 -89.5529>
  }]
 R_back_jacket_02 [{
   shapeType=Capsule
   length=8.3207
   radius=3.0138
   shapeName=R_back_jacket_02_capsule
   center=<3.5604 -0.3184 -0.0114>
   rotation=<rot -14.5194 -96.1769 -2.3129>
  }]
 R_back_jacket_03 [{
   shapeType=Capsule
   length=6.1056
   radius=3.2241
   shapeName=R_back_jacket_03_capsule
   center=<3.86 -1.3746 -0.2852>
   rotation=<rot -12.634 -90 -1.2398>
  }]
 R_front_jacket_02 [{
   shapeType=Capsule
   length=6.9523
   radius=3.0739
   shapeName=R_front_jacket_02_capsule
   center=<3.9277 1.0411 -0.1743>
   rotation=<rot -4.9255 8.1276 -82.9961>
  }]
 R_front_jacket_03 [{
   shapeType=Capsule
   length=4.9956
   radius=3.3811
   shapeName=R_front_jacket_03_capsule
   center=<3.7099 0.6734 1.1145>
   rotation=<rot -1.1009 6.2768 -86.17>
  }]
 R_side_jacket_02 [{
   shapeType=Capsule
   length=6.9228
   radius=4.3578
   shapeName=R_side_jacket_02_capsule
   center=<3.9873 1.1828 -0.0191>
   rotation=<rot 0.2084 -168.2465 -87.3789>
  }]
 calf_l [{shapeType=Capsule, length=23.6827, radius=14.1606, center=<-18.842 -1.0215 0.6738>, rotation=<rot 0 -96.3602 -89.1658>}]
 calf_r [{shapeType=Capsule, length=23.6824, radius=14.1073, center=<18.8469 1.073 -0.7373>, rotation=<rot 0 -96.3646 -89.1648>}]
 foot_l [{shapeType=Capsule, length=14.0567, radius=6.9845, center=<-7.0194 6.0721 1.3988>, rotation=<rot 0 -169.3263 -82.2503>}]
 foot_r [{shapeType=Capsule, length=14.0567, radius=6.9845, center=<7.0194 -6.0721 -1.3988>, rotation=<rot 0 -169.3263 -82.2503>}]
 hand_l [{shapeType=Capsule, length=9.0184, radius=7.8212, center=<8.164 -2.0814 -0.3272>, rotation=<rot 0 97.1737 -55.0078>}]
 hand_r [{shapeType=Capsule, length=9.0184, radius=7.8213, center=<-8.164 2.0814 0.3272>, rotation=<rot 0 97.1737 -55.0078>}]
 lowerarm_l [{shapeType=Capsule, length=13.4226, radius=6.0859, center=<15.0716 0.7901 0.3677>, rotation=<rot 0 90.7156 -85.3738>}]
 lowerarm_r [{shapeType=Capsule, length=13.4226, radius=6.0859, center=<-15.0716 -0.7901 -0.3677>, rotation=<rot 0 90.7156 -85.3739>}]
 pelvis [{shapeType=Capsule, length=12.6188, radius=10.5046, center=<-2.4015 -0.3568 -0.0015>, rotation=<rot -0.087 -90 -0.0729>}]
 spine_04 [{shapeType=Capsule, length=4.2659, radius=10.8094, center=<0 1.5164 0>}]
 thigh_l [{shapeType=Capsule, length=25.2859, radius=11.2092, center=<-22.1127 -1.4904 -2.8733>, rotation=<rot 0 -89.3823 -82.9293>}]
 thigh_r [{shapeType=Capsule, length=25.2859, radius=11.2092, center=<22.1127 1.4904 2.8733>, rotation=<rot 0 -89.3823 -82.9293>}]
 upperarm_l [{shapeType=Capsule, length=17.1947, radius=8.2119, center=<18.1096 -0.5806 -1.4166>, rotation=<rot 0 92.6775 -86.427>}]
 upperarm_r [{shapeType=Capsule, length=17.2444, radius=8.1659, center=<-18.0786 0.6302 1.2328>, rotation=<rot 0 92.4507 -85.6855>}]
}
constraints [
 {bone1Name=spine_04, bone2Name=pelvis, swing1LimitDegrees=45, swing2LimitDegrees=45, twistLimitDegrees=45}
 {bone1Name=upperarm_l, bone2Name=spine_04, swing1LimitDegrees=45, swing2LimitDegrees=45, twistLimitDegrees=45}
 {bone1Name=lowerarm_l, bone2Name=upperarm_l, swing1LimitDegrees=45, swing2LimitDegrees=45, twistLimitDegrees=45}
 {bone1Name=hand_l, bone2Name=lowerarm_l, swing1LimitDegrees=45, swing2LimitDegrees=45, twistLimitDegrees=45}
 {bone1Name=upperarm_r, bone2Name=spine_04, swing1LimitDegrees=45, swing2LimitDegrees=45, twistLimitDegrees=45}
 {bone1Name=lowerarm_r, bone2Name=upperarm_r, swing1LimitDegrees=45, swing2LimitDegrees=45, twistLimitDegrees=45}
 {bone1Name=hand_r, bone2Name=lowerarm_r, swing1LimitDegrees=45, swing2LimitDegrees=45, twistLimitDegrees=45}
 {bone1Name=thigh_l, bone2Name=pelvis, swing1LimitDegrees=45, swing2LimitDegrees=45, twistLimitDegrees=45}
 {bone1Name=calf_l, bone2Name=thigh_l, swing1LimitDegrees=45, swing2LimitDegrees=45, twistLimitDegrees=45}
 {bone1Name=foot_l, bone2Name=calf_l, swing1LimitDegrees=45, swing2LimitDegrees=45, twistLimitDegrees=45}
 {bone1Name=thigh_r, bone2Name=pelvis, swing1LimitDegrees=45, swing2LimitDegrees=45, twistLimitDegrees=45}
 {bone1Name=calf_r, bone2Name=thigh_r, swing1LimitDegrees=45, swing2LimitDegrees=45, twistLimitDegrees=45}
 {bone1Name=foot_r, bone2Name=calf_r, swing1LimitDegrees=45, swing2LimitDegrees=45, twistLimitDegrees=45}
 {
  bone1Name=BASE_L_back_jacket_03
  bone2Name=spine_04
  swing1LimitDegrees=45
  swing1Motion=Free
  swing2LimitDegrees=45
  swing2Motion=Free
  twistLimitDegrees=45
  twistMotion=Free
 }
 {bone1Name=L_back_jacket_03, bone2Name=BASE_L_back_jacket_03, swing1LimitDegrees=15, swing2LimitDegrees=5, twistLimitDegrees=5}
 {bone1Name=L_back_jacket_02, bone2Name=L_back_jacket_03, swing1LimitDegrees=15, swing2LimitDegrees=5, twistLimitDegrees=5}
 {
  bone1Name=BASE_L_front_jacket_03
  bone2Name=spine_04
  swing1LimitDegrees=45
  swing1Motion=Free
  swing2LimitDegrees=45
  swing2Motion=Free
  twistLimitDegrees=45
  twistMotion=Free
 }
 {bone1Name=L_front_jacket_03, bone2Name=BASE_L_front_jacket_03, swing1LimitDegrees=15, swing2LimitDegrees=20, twistLimitDegrees=10}
 {bone1Name=L_front_jacket_02, bone2Name=L_front_jacket_03, swing1LimitDegrees=15, swing2LimitDegrees=20, twistLimitDegrees=10}
 {
  bone1Name=BASE_L_side_jacket_02
  bone2Name=spine_04
  swing1LimitDegrees=45
  swing1Motion=Free
  swing2LimitDegrees=45
  swing2Motion=Free
  twistLimitDegrees=45
  twistMotion=Free
 }
 {bone1Name=L_side_jacket_02, bone2Name=BASE_L_side_jacket_02, swing1LimitDegrees=15, swing2LimitDegrees=20, twistLimitDegrees=10}
 {
  bone1Name=BASE_R_back_jacket_03
  bone2Name=spine_04
  swing1LimitDegrees=45
  swing1Motion=Free
  swing2LimitDegrees=45
  swing2Motion=Free
  twistLimitDegrees=45
  twistMotion=Free
 }
 {bone1Name=R_back_jacket_03, bone2Name=BASE_R_back_jacket_03, swing1LimitDegrees=15, swing2LimitDegrees=5, twistLimitDegrees=5}
 {bone1Name=R_back_jacket_02, bone2Name=R_back_jacket_03, swing1LimitDegrees=15, swing2LimitDegrees=5, twistLimitDegrees=5}
 {
  bone1Name=BASE_R_front_jacket_03
  bone2Name=spine_04
  swing1LimitDegrees=45
  swing1Motion=Free
  swing2LimitDegrees=45
  swing2Motion=Free
  twistLimitDegrees=45
  twistMotion=Free
 }
 {bone1Name=R_front_jacket_03, bone2Name=BASE_R_front_jacket_03, swing1LimitDegrees=15, swing2LimitDegrees=20, twistLimitDegrees=10}
 {bone1Name=R_front_jacket_02, bone2Name=R_front_jacket_03, swing1LimitDegrees=15, swing2LimitDegrees=20, twistLimitDegrees=10}
 {
  bone1Name=BASE_R_side_jacket_02
  bone2Name=spine_04
  swing1LimitDegrees=45
  swing1Motion=Free
  swing2LimitDegrees=45
  swing2Motion=Free
  twistLimitDegrees=45
  twistMotion=Free
 }
 {bone1Name=R_side_jacket_02, bone2Name=BASE_R_side_jacket_02, swing1LimitDegrees=15, swing2LimitDegrees=20, twistLimitDegrees=10}
]
