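# Control Rig for the BadCat face accessory, animating a small cat with two joysticks from the player's head pose and movement.
# Cat neck and head take rotation from the player's neck and head bones, with two spherical pose readers nudging the cat head at certain head angles.
# Four smoothed directional movement variables blend cat arm offsets, then joysticks aim at the hands.
# Skydiving, parachute and emote variables are declared but unused.
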ControlRigBlueprint /BRCosmetics/Characters/Player/Male/Large/Bodies/M_LRG_BadCat/Meshes/Parts/M_LRG_BadCat_FaceAcc_CtrlRig
graphs [
 RigVMFunctionLibrary {links [], nodes []}
 RigVMModel {nodes [
   RigUnit_BeginExecution = &RigUnit_BeginExecution()
   Multiply = &RigVMFunction_MathTransformMul(
    B=&RigUnit_GetTransform(Item=(Type=Bone,Name="neck_02"), Space=LocalSpace, bInitial=False).Transform
    A=&RigUnit_GetTransform(Item=(Type=Bone,Name="neck_01"), Space=LocalSpace, bInitial=False).Transform
   ) {template="Multiply::Execute(in A,in B,out Result)"}
   &RigUnit_ModifyTransforms(
    ItemToModify=((Item=(Type=Bone,Name="cat_neck"),Transform=(Rotation=(X=0,Y=0,Z=0,W=1),Translation=(X=0,Y=0,Z=0),Scale3D=(X=1,Y=1,Z=1))),(Item=(Type=Bone,Name="cat_head"),Transform=(Rotation=(X=0,Y=0,Z=0,W=1),Translation=(X=0,Y=0,Z=0),Scale3D=(X=1,Y=1,Z=1))))
    Weight=1
    ItemToModify.0.Transform.Rotation=@Multiply.Result.Rotation
    ExecutePin=@RigUnit_BeginExecution.ExecutePin
    ItemToModify.1.Transform.Rotation=&RigUnit_GetTransform(Item=(Type=Bone,Name="head"), Space=LocalSpace, bInitial=False).Transform.Rotation
   )
   > n1 = &RigUnit_SphericalPoseReader(
    ActiveRegionScaleFactors=(PositiveWidth=1,NegativeWidth=1,PositiveHeight=1,NegativeHeight=1)
    ActiveRegionSize=0.1
    Debug=(bDrawDebug=True,bDraw2D=False,bDrawLocalAxes=False,DebugScale=25,DebugSegments=20,DebugThickness=0.25)
    DriverAxis=(X=1,Y=0,Z=0)
    DriverItem=(Type=Bone,Name="head")
    FalloffRegionScaleFactors=(PositiveWidth=1.5,NegativeWidth=1.5,PositiveHeight=1.5,NegativeHeight=1.5)
    FalloffSize=0.22
    FlipHeightScaling=False
    FlipWidthScaling=False
    OptionalParentItem=(Type=Bone,Name="None")
    RotationOffset=(X=82.7523,Y=-19.6663,Z=-114.9733)
   )
   > &RigUnit_ModifyTransforms(
    ItemToModify=((Item=(Type=Bone,Name="cat_head"),Transform=(Rotation=(X=0,Y=0,Z=0,W=1),Translation=(X=-1,Y=0,Z=0),Scale3D=(X=1,Y=1,Z=1))))
    WeightMinimum=0
    Weight=@n1.OutputParam
   )
   > n2 = &RigUnit_SphericalPoseReader(
    ActiveRegionScaleFactors=(PositiveWidth=0.7,NegativeWidth=0.7,PositiveHeight=0.7,NegativeHeight=0.7)
    ActiveRegionSize=0.1
    Debug=(bDrawDebug=True,bDraw2D=False,bDrawLocalAxes=False,DebugScale=25,DebugSegments=20,DebugThickness=0.25)
    DriverAxis=(X=1,Y=0,Z=0)
    DriverItem=(Type=Bone,Name="head")
    FalloffRegionScaleFactors=(PositiveWidth=1,NegativeWidth=1,PositiveHeight=1,NegativeHeight=1)
    FalloffSize=0.2
    FlipHeightScaling=False
    FlipWidthScaling=False
    OptionalParentItem=(Type=Bone,Name="None")
    RotationOffset=(X=95.3185,Y=-19.6663,Z=-75.4849)
   )
   > &RigUnit_ModifyTransforms(
    ItemToModify=((Item=(Type=Bone,Name="cat_head"),Transform=(Rotation=(X=0,Y=0,Z=0,W=1),Translation=(X=1,Y=-2,Z=0),Scale3D=(X=1,Y=1,Z=1))))
    WeightMinimum=0
    Weight=@n2.OutputParam
   )
   AlphaInterp = &RigVMFunction_AlphaInterp(
    InterpSpeedDecreasing=10
    InterpSpeedIncreasing=10
    Scale=1
    bClampResult=False
    Value=RigVMVariableNode() {variable=Character Forward}.Value
   )
   > &RigUnit_ModifyTransforms(
    ItemToModify="((Item=(Type=Bone,Name=\"cat_elbow_l\"),Transform=(Rotation=(X=0,Y=0.1736,Z=0,W=0.9848),Translation=(X=1.4314,Y=0,Z=0.0922),Scale3D=(X=1,Y=1,Z=1))),(Item=(Type=Bone,Name=\"cat_hand_l\"),Transform=(Rotation=(X=0,Y=-0.0872,Z=0,W=0.9962),Translation=(X=-0.2375,Y= ...(truncated)"
    WeightMinimum=0
    Weight=@AlphaInterp.Result
   )
   n3 = &RigVMFunction_AlphaInterp(
    InterpSpeedDecreasing=10
    InterpSpeedIncreasing=10
    Scale=1
    bClampResult=False
    Value=RigVMVariableNode() {variable=Character Backward}.Value
   )
   > &RigUnit_ModifyTransforms(
    ItemToModify="((Item=(Type=Bone,Name=\"cat_elbow_l\"),Transform=(Rotation=(X=0,Y=-0.0872,Z=0,W=0.9962),Translation=(X=-2.2829,Y=0,Z=-1.8578),Scale3D=(X=1,Y=1,Z=1))),(Item=(Type=Bone,Name=\"cat_hand_l\"),Transform=(Rotation=(X=0,Y=0,Z=0,W=1),Translation=(X=0,Y=0,Z=0),Scale3D ...(truncated)"
    WeightMinimum=0
    Weight=@n3.Result
   )
   n4 = &RigVMFunction_AlphaInterp(
    InterpSpeedDecreasing=10
    InterpSpeedIncreasing=10
    Scale=1
    bClampResult=False
    Value=RigVMVariableNode() {variable=Character Move Right}.Value
   )
   > &RigUnit_ModifyTransforms(
    ItemToModify="((Item=(Type=Bone,Name=\"cat_elbow_l\"),Transform=(Rotation=(X=0,Y=0.0872,Z=0,W=0.9962),Translation=(X=2.366,Y=0,Z=-0.112),Scale3D=(X=1,Y=1,Z=1))),(Item=(Type=Bone,Name=\"cat_hand_l\"),Transform=(Rotation=(X=0,Y=0.0872,Z=0,W=0.9962),Translation=(X=-0.0504,Y=0, ...(truncated)"
    WeightMinimum=0
    Weight=@n4.Result
   )
   n5 = &RigVMFunction_AlphaInterp(
    InterpSpeedDecreasing=10
    InterpSpeedIncreasing=10
    Scale=1
    bClampResult=False
    Value=RigVMVariableNode() {variable=Character Move Left}.Value
   )
   > &RigUnit_ModifyTransforms(
    ItemToModify="((Item=(Type=Bone,Name=\"cat_elbow_l\"),Transform=(Rotation=(X=0,Y=-0.0872,Z=0,W=0.9962),Translation=(X=-2.6599,Y=0,Z=-0.469),Scale3D=(X=1,Y=1,Z=1))),(Item=(Type=Bone,Name=\"cat_hand_l\"),Transform=(Rotation=(X=0,Y=-0.1736,Z=0,W=0.9848),Translation=(X=-0.7889, ...(truncated)"
    WeightMinimum=0
    Weight=@n5.Result
   )
   > &RigUnit_AimConstraintLocalSpaceOffset(
    Child=(Type=Bone,Name="joystick_l")
    Filter=(bX=True,bY=True,bZ=True)
    Parents=((Item=(Type=Bone,Name="cat_hand_l"),Weight=1))
    Weight=1
    WorldUp=(Target=(X=1,Y=0,Z=0),Kind=Direction,Space=(Type=None,Name="head"))
   )
   > &RigUnit_AimConstraintLocalSpaceOffset(
    Child=(Type=Bone,Name="joystick_r")
    Filter=(bX=True,bY=True,bZ=True)
    Parents=((Item=(Type=Bone,Name="cat_hand_r"),Weight=1))
    Weight=1
    WorldUp=(Target=(X=1,Y=0,Z=0),Kind=Direction,Space=(Type=None,Name="None"))
   )
  ]}
]
hierarchy [
 root {loc=<0 0 0>, rot=<0 0 0 1>}
 pelvis {parent=root, loc=<0 -0.7321 92.4218>, rot=<0.0127 0.707 -0.0127 -0.707>}
 spine_01 {parent=pelvis, loc=<5.6002 -0.5128 0>, rot=<0 0 0.034 -0.9994>}
 spine_02 {parent=spine_01, loc=<8.0756 0.0139 0>, rot=<0 0 0.0238 -0.9997>}
 spine_03 {parent=spine_02, loc=<7.0395 0 0>, rot=<0 0 0.0149 -0.9999>}
 spine_04 {parent=spine_03, loc=<8.1023 0 0>, rot=<0 0 -0.0132 -0.9999>}
 spine_05 {parent=spine_04, loc=<10.0141 0 0>, rot=<0 0 -0.0015 -1>}
 neck_01 {parent=spine_05, loc=<10.815 1.9196 0>, rot=<0 0 -0.1864 -0.9825>}
 neck_02 {parent=neck_01, loc=<4.7347 0.3045 0>, rot=<0 0 0.0363 -0.9993>}
 head {parent=neck_02, loc=<5.7632 0.5552 0>, rot=<0 0 0.115 -0.9934>}
 joystick_r {parent=spine_05, loc=<10.6526 8.6516 2.7761>, rot=<-0.7065 -0.0284 -0.0284 -0.7065>}
 joystick_l {parent=spine_05, loc=<10.6526 8.6516 -2.7761>, rot=<-0.7065 -0.0284 -0.0284 -0.7065>}
 spine_05_scale {parent=spine_05, loc=<0 0 0>, rot=<0 0 0 1>}
 cat_neck {parent=spine_05_scale, loc=<16.2555 0.2622 0>, rot=<0 0 -0.1506 -0.9886>}
 cat_head {parent=cat_neck, loc=<3.0531 0.5552 0>, rot=<0 0 0.115 -0.9934>}
 cat_elbow_l {parent=spine_05_scale, loc=<13.3437 4.3642 -3.4493>, rot=<0.5553 0.4492 0.4985 0.4913>}
 cat_hand_l {parent=cat_elbow_l, loc=<4.608 0 0>, rot=<0 0 0 1>}
 cat_elbow_r {parent=spine_05_scale, loc=<13.3437 4.3642 3.449>, rot=<-0.4913 -0.4985 -0.4492 -0.5553>}
 cat_hand_r {parent=cat_elbow_r, loc=<4.6041 0 0>, rot=<0 0 0 1>}
 flipbook_face_bypass_mapping {type=Curve, loc=<0 0 0>, rot=<0 0 0 1>}
 flipbook_face_L_eye_index {type=Curve, loc=<0 0 0>, rot=<0 0 0 1>}
 flipbook_face_mouth_index {type=Curve, loc=<0 0 0>, rot=<0 0 0 1>}
 flipbook_face_R_eye_index {type=Curve, loc=<0 0 0>, rot=<0 0 0 1>}
 flipbook_face_L_brow_index {type=Curve, loc=<0 0 0>, rot=<0 0 0 1>}
 flipbook_face_R_brow_index {type=Curve, loc=<0 0 0>, rot=<0 0 0 1>}
]
variables [
 Character Backward {type=double, default=0}
 Character Forward {type=double, default=0}
 Character Move Left {type=double, default=0}
 Character Move Right {type=double, default=0}
 ParachuteOpened {type=bool, default=False}
 PlayingEmote {type=bool, default=False}
 skydiving {type=bool, default=False}
]
