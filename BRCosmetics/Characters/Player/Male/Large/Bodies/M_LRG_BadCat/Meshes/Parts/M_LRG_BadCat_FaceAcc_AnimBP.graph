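# Animation Blueprint for the M_LRG_BadCat face accessory part, posing it from the attached parent mesh and character movement.
# Copy Pose From Mesh, including curves, feeds a Control Rig driven by four directional movement values and skydiving, parachute and emote flags.
# Each update maps PlayerMovement output into clamped directions, gated on a valid pawn and a hardcoded Skirmish backpack mesh reference.
# In the front end it also caches the Frontend_Idle_Bypass curve, which the AnimGraph leaves unused.
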AnimBlueprint /BRCosmetics/Characters/Player/Male/Large/Bodies/M_LRG_BadCat/Meshes/Parts/M_LRG_BadCat_FaceAcc_AnimBP
target_skeleton=&/M_LRG_BadCat_FaceAcc_Skeleton
graphs [
 AnimGraph {nodes [
   n1 = Control Rig!(
    Source=Copy Pose From Mesh() {bCopyCurves=true, bUseAttachedParent=true}
    Alpha
    Character Forward:double=Get Character Forward() -> (Character Forward:double)
    Character Backward:double=Get Character Backward() -> (Character Backward:double)
    Character Move Right:double=Get Character Move Right() -> (Character Move Right:double)
    Character Move Left:double=Get Character Move Left() -> (Character Move Left:double)
    skydiving:bool=Get bIsSkyDiving()
    ParachuteOpened:bool=Get bIsParachuteOpened()
    PlayingEmote:bool=Get bIsPlayingEmote()
   ) -> (Pose) {
    alphaBoolBlend {blendOption=Linear}
    controlRigAssetReference {blueprintRigClass=&/M_LRG_BadCat_FaceAcc_CtrlRig.M_LRG_BadCat_FaceAcc_CtrlRig_C}
   }
   Output Pose(Result=@n1) {layerGroup=None}
  ]}
 EventGraph {nodes [
   Event BlueprintInitializeAnimation()
   > Set SkelMesh(SkelMesh:Skeletal Mesh Object Reference=/Game/Accessories/FORT_Backpacks/Backpack_F_MED_Skirmish/Meshes/F_MED_Skirmish_Pack) -> (Output_Get:Skeletal Mesh Object Reference)
   > n1 = Cast To FortPlayerPawn(Object=TryGetFortPawn())
   > Set Pawn(Pawn:Fort Player Pawn Object Reference=@n1) -> (Output_Get:Fort Player Pawn Object Reference)
   Event BlueprintUpdateAnimation()
   > Get!(execute:Exec) -> (then:Exec, else:Exec, SkelMesh:Skeletal Mesh Object Reference)
   > Get!(execute) -> (then, else, Pawn:Fort Player Pawn Object Reference)
   > n2 = Sequence()
   > n3 = PlayerMovement(AnimInstance=Self-Reference(), XDotProductCoefficient=25, YDotProductCoefficient=25, ZDotProductCoefficient=1)
   > Set Character Move Right(Character Move Right:double=MapRangeClamped(Value=@n3, InRangeA=0, InRangeB=25)) -> (Output_Get:double)
   > Set Character Move Left(Character Move Left:double=MapRangeClamped(Value=@n3, InRangeA=0, InRangeB=-25)) -> (Output_Get:double)
   > Set Character Forward(Character Forward:double=MapRangeClamped(Value=@n3.YMovement, InRangeA=0, InRangeB=25)) -> (Output_Get:double)
   > Set Character Backward(Character Backward:double=MapRangeClamped(Value=@n3.YMovement, InRangeA=0, InRangeB=-25)) -> (Output_Get:double)
   Branch(execute=@n2.then_1, Condition=Get bIsInFrontEnd())
   > Set IsPlayingFrontendIdle(IsPlayingFrontendIdle:float=GetCurveValue(self=GetOwnerFrontEndAnimInstance(), CurveName=Frontend_Idle_Bypass)) -> (Output_Get:float)
  ]}
]
variables [
 Character Backward {type=number}
 Character Forward {type=number}
 Character Move Left {type=number}
 Character Move Right {type=number}
 flipbook_face_activation_var {type=FN_Flipbook_Face_Region_Activation, default {enable_Brow=true, enable_Eye=true, enable_Mouth=true}}
 IsPlayingFrontendIdle {type=number}
 Pawn {type=FortPlayerPawn}
 SkelMesh {type=SkeletalMesh, default=null}
]
